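; Scan actor blueprint for the ToyFox Scan ability that fires a pulse-indicator scan as soon as it spawns.
; On BeginPlay it assigns its indicating controller to its PulseIndicator component, then starts the pulse markers.
; It logs which player performed the scan and sets a one-second lifespan, so it acts as a short-lived, one-shot actor.
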(event EventBeginPlay
  (bind _pulseindicator (Variables|Default|GetPulseIndicator))
  (bind _indicatingcontroller (Variables|Default|GetIndicatingController))
  ;[0 0.15 0.03623 0.5] BeginPlay: Perform scan right away.
  (|CallParentFunction)
  (Class|PulseIndicatorComponent|SetIndicatingController _pulseindicator _indicatingcontroller)
  (Class|PulseIndicatorComponent|StartPulseMarkers _pulseindicator)
  (Fortnite|FortShippingLog self (Utilities|String|Append "BeginPlay: Scan performed for player " (Utilities|GetDisplayName _indicatingcontroller)))
  (Actor|SetLifeSpan 1.0))

(fn ConstructionScript)
