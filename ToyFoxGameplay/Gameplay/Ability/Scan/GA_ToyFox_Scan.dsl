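; Gameplay ability for the ToyFox giant player pawn that fires a scan pulse for nearby enemy players.
; On event activation it commits on authority, applies a recently-used effect and finds the ground below the pawn, cancelling if the avatar is another pawn type.
; The server spawns B_ToyFox_ScanActor there, and a pulse cue carries the hotfixable radius as raw magnitude.
; It also emits a noise event for AI perception, then ends immediately.
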(event Ability|EventActivateAbilityFromEvent
  (params (EventData :type Gameplay%20Event%20Data%20Structure%20%28by%20ref%29))
  ;[0 0.15 0.02057 0.5] ActivateAbilityFromEvent: Perform scan for nearby enemy players.
  (|CallParentFunction EventData)
  (Utilities|FlowControl|Sequence
    (:then_0
      (Ability|HasAuthority
        (:True
          (Ability|CommitAbility))))
    (:then_1
      (Ability|ApplyGameplayEffectToOwner (Variables|Settings|GetScanRecentlyUsedGameplayEffectClass))
      (bind _asb_toy_fox_giant_player_pawn (Utilities|Casting|CastToB_ToyFoxGiantPlayerPawn (CallFunction|GetAvatarasFortPlayerPawn))
        (:then
          (bind _returnvalue (Helpers|FindStaticGroundLocationUsingWorldContextAt self (Transformation|GetActorLocation _asb_toy_fox_giant_player_pawn) _asb_toy_fox_giant_player_pawn (.z (Transformation|GetActorLocation _asb_toy_fox_giant_player_pawn))))
          (if (Variables|Debug|GetbDebug)
            (Rendering|Debug|DrawDebugSphere _returnvalue 1.5e+02 12 "(R=0.005557,G=0.128205,B=1.000000,A=1.000000)" 5.0 3.0)
            (else
              _))
          (Ability|HasAuthority
            (:True
              (Game|SpawnActorfromClass "/ToyFoxGameplay/Gameplay/Ability/Scan/B_ToyFox_ScanActor.B_ToyFox_ScanActor_C" (Math|Conversions|ToTransform(Vector) _returnvalue) "AlwaysSpawn" "MultiplyWithRoot" _asb_toy_fox_giant_player_pawn (Utilities|Casting|CastToFortPlayerControllerAthena (Fortnite|GetFortPlayerControllerfromActor _asb_toy_fox_giant_player_pawn)) _asb_toy_fox_giant_player_pawn))
            (:False
              _))
          (Ability|ExecuteGameplayCueWithParamsOnOwner (Utilities|Struct|BreakGameplayCueTag (Variables|GameplayCueTag|GetGameplayCueTagScanPulse)) (Ability|GameplayCue|MakeGameplayCueParameters :NormalizedMagnitude 0.0 :RawMagnitude (Math|Float|MakeLiteralFloat (ScalableFloat|GetCachedScalableFloatValue (Variables|Hotfix|GetHotfix_ScanRadius))) :Location _returnvalue :Normal "0,0,0" :SourceObject _asb_toy_fox_giant_player_pawn :GameplayEffectLevel 1 :AbilityLevel 1 :bReplicateLocationWhenUsingMinimalRepProxy false))
          (Fortnite|FortShippingLog self "ActivateAbility: Scan performed."))
        (:CastFailed
          (Ability|CancelAbility))))
    (:then_2
      ;;[0.03689 0.04106 0.15 0.5] Add noise event for AI to react to
      (bind _outfortplayerpawn (CallFunction|GetAvatarasValidFortPlayerPawn)
        (:ValidFortPlayerPawn
          (AI|Perception|MakeNoiseEventatLocation _outfortplayerpawn (ScalableFloat|GetValueatLevel (Variables|Hotfix|GetHotfix_PerceptionRange)) (Transformation|GetActorLocation _outfortplayerpawn) "WeaponFiring"))))
    (:then_3
      (Ability|EndAbility))))

(fn CanActivateAbility
  (params (ActorInfo :type Gameplay%20Ability%20Actor%20Info%20Structure) (Handle :type Gameplay%20Ability%20Spec%20Handle%20Structure))
  (return 0 true))
